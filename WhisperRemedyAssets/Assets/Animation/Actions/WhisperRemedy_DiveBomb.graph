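# Level Sequence for the WhisperRemedy DiveBomb action animation, built around a single actor binding with no root-level tracks.
# The binding is labeled BP_ShirtTilapia_C but references the BP_WhisperRemedy_C class, and carries a spawn track and a transform track at identity defaults.
# An MR_ShirtTilapia Control Rig track provides rig-driven posing, alongside an attach section with a bounded frame range.
# Display rate is 30 fps.
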LevelSequence /WhisperRemedyAssets/Assets/Animation/Actions/WhisperRemedy_DiveBomb
bindings [BP_ShirtTilapia_C {
  id=#1
  possessed_class=BP_WhisperRemedy_C
  tracks [
   {type=MovieSceneSpawnTrack, sections [{type=MovieSceneSpawnSection, channels [None {type=MovieSceneScriptingBoolChannel, default=true}]}]}
   {
    type=MovieScene3DTransformTrack
    sections [{
      type=MovieScene3DTransformSection
      channels [
       Location.X {type=MovieSceneScriptingDoubleChannel, default=0}
       Location.Y {type=MovieSceneScriptingDoubleChannel, default=0}
       Location.Z {type=MovieSceneScriptingDoubleChannel, default=0}
       Rotation.X {type=MovieSceneScriptingDoubleChannel, default=0}
       Rotation.Y {type=MovieSceneScriptingDoubleChannel, default=0}
       Rotation.Z {type=MovieSceneScriptingDoubleChannel, default=0}
       Scale.X {type=MovieSceneScriptingDoubleChannel, default=1}
       Scale.Y {type=MovieSceneScriptingDoubleChannel, default=1}
       Scale.Z {type=MovieSceneScriptingDoubleChannel, default=1}
      ]
     }]
   }
   {type=MovieSceneControlRigParameterTrack, display_name=MR_ShirtTilapia, sections [{type=MovieSceneControlRigParameterSection}]}
   {type=MovieScene3DAttachTrack, sections [{type=MovieScene3DAttachSection, end_frame=198, start_frame=2}]}
  ]
 }]
display_rate {denominator=1, numerator=30}
tick_resolution {denominator=1, numerator=24000}
tracks []
